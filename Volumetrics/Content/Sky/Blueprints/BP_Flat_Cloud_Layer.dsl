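; Flat volumetric cloud layer actor that renders clouds through a material on a sky sphere mesh.
; The construction script creates a MID from the chosen material, assigns it to the sphere, and sets sphere visibility and translucent sort priority.
; It pushes shape, detail, density, falloff and step-size variables into the MID, converting kilometres to centimetres and packing scale, bias and offset into one vector.
; BeginPlay and overlap events are empty.
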(event EventBeginPlay)

(event Collision|EventActorBeginOverlap
  (params (OtherActor :type Actor%20Object%20Reference)))

(fn ConstructionScript
  (bind (_x _y) (Math|Vector2D|BreakVector2D (Variables|Default|GetOffset)))
  (bind _cloud_scale_km (Variables|Default|GetCloud%20Scale%20Km))
  (bind _returnvalue (* _cloud_scale_km 1e+05))
  (bind _mid (Variables|Debug|GetMID))
  (bind _mid_2 (Variables|Debug|GetMID))
  (bind _mid_3 (Variables|Debug|GetMID))
  (Class|VolumetricsBPLIbrary|CreateAndAssignMID (Variables|Debug|GetMID) (Variables|Default|GetMaterial))
  (Rendering|Material|SetMaterial (Variables|Default|GetEditorSkySphere) 0 (Variables|Debug|GetMID))
  (Rendering|SetVisibility (Variables|Default|GetEditorSkySphere) (Variables|Default|GetShow%20Sky%20Dome) false)
  (Rendering|Material|SetScalarParameterValue _mid "Bias" (Variables|Default|GetCloud%20Bias) :!class /Script/Engine.MaterialInstanceDynamic)
  (Rendering|Material|SetScalarParameterValue _mid "CloudDensity" (Variables|Default|GetCloud%20Density) :!class /Script/Engine.MaterialInstanceDynamic)
  (Rendering|Material|SetScalarParameterValue _mid "Layer1 Alt" (* (Variables|Default|GetLayer%20Altitude%20Km) 1e+05) :!class /Script/Engine.MaterialInstanceDynamic)
  (Rendering|Material|SetScalarParameterValue _mid "Layer1 Thickness" (* (Variables|Default|GetLayer%20Thickness%20Km) 1e+05) :!class /Script/Engine.MaterialInstanceDynamic)
  (Rendering|Material|SetScalarParameterValue _mid "CloudDetail1Strength" (Variables|Default|GetCloud%20Detail%20Amount) :!class /Script/Engine.MaterialInstanceDynamic)
  (Rendering|Material|SetScalarParameterValue _mid "CloudDetail2Strength" (Variables|Default|GetCloud%20Detail%20Amount%202) :!class /Script/Engine.MaterialInstanceDynamic)
  (Rendering|Material|SetScalarParameterValue _mid "Earth Scale" (Variables|Default|GetAtmosphere%20Scale%20Relative%20to%20Earth) :!class /Script/Engine.MaterialInstanceDynamic)
  (Rendering|Material|SetScalarParameterValue _mid "StepSize" (Variables|Default|GetStepsize) :!class /Script/Engine.MaterialInstanceDynamic)
  (Rendering|Material|SetScalarParameterValue _mid "CloudDetail1Scale" (Variables|Default|GetCloud%20Detail%20Scale) :!class /Script/Engine.MaterialInstanceDynamic)
  (Rendering|Material|SetScalarParameterValue _mid "CloudDetail2Scale" (Variables|Default|GetCloud%20Detail%20Scale%202) :!class /Script/Engine.MaterialInstanceDynamic)
  (Rendering|Material|SetScalarParameterValue _mid "ShadowDensity" (Variables|Default|GetShadow%20Density) :!class /Script/Engine.MaterialInstanceDynamic)
  (Rendering|Material|SetVectorParameterValue _mid_2 "CloudBaseParams" (Utilities|Struct|MakeLinearColor _returnvalue (Variables|Default|GetCloud%20Bias) _x _y) :!class /Script/Engine.MaterialInstanceDynamic)
  (Rendering|Material|SetScalarParameterValue _mid_2 "Cloud Scale" _returnvalue :!class /Script/Engine.MaterialInstanceDynamic)
  (Rendering|Material|SetScalarParameterValue _mid_3 "Upper Falloff" (Variables|Default|GetUpper%20Falloff) :!class /Script/Engine.MaterialInstanceDynamic)
  (Rendering|Material|SetScalarParameterValue _mid_3 "Lower Falloff" (Variables|Default|GetLower%20Falloff) :!class /Script/Engine.MaterialInstanceDynamic)
  (Rendering|SetTranslucentSortPriority (Variables|Default|GetEditorSkySphere) (Variables|Default|GetSort)))
