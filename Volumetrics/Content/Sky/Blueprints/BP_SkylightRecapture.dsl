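; Blueprint actor that keeps a referenced skylight's captured sky lighting in sync with the scene in the volumetrics sky setup.
; Every tick, if the Skylight reference is valid, it recaptures the sky and writes the Sun actor's forward vector to SunVector in the MPC_Planet material parameter collection.
; A Single Capture custom event performs a one-off recapture without updating the sun vector.
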(event EventTick
  (params (DeltaSeconds :type Float%20%28single-precision%29))
  (bind _skylight (Variables|Default|GetSkylight))
  (Utilities|IsValid _skylight
    (:Is%20Valid
      (Rendering|Components|SkyLight|RecaptureSky (Variables|/Script/Engine.SkyLight|GetLightComponent _skylight))
      (Rendering|Material|SetVectorParameterValue "/Volumetrics/Content/PlanetPrototype/Materials/MPC_Planet.MPC_Planet" "SunVector" (Math|Conversions|ToLinearColor(Vector) (Math|Rotator|GetRotationXVector (Transformation|GetActorRotation (Variables|Default|GetSun)))) :!class /Script/Engine.KismetMaterialLibrary))))

(event Custom|Single%20Capture
  (bind _skylight (Variables|Default|GetSkylight))
  (Utilities|IsValid _skylight
    (:Is%20Valid
      (Rendering|Components|SkyLight|RecaptureSky (Variables|/Script/Engine.SkyLight|GetLightComponent _skylight)))))

(fn ConstructionScript)
