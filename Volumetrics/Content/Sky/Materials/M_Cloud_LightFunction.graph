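# Light function material that, by its name and texture, is meant to pattern a light with cloud shadows.
# It samples the engine fog-sheet cloud texture using world XY divided by a large scale parameter, oddly named "fasdf".
# The UVs are shifted by SunVector XY over its Z from MPC_Planet, scaled by parameter Z, so the pattern follows sun angle.
# A SkyAtmosphereLightDirection node is present but unused.
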Material /Volumetrics/Content/Sky/Materials/M_Cloud_LightFunction
graph {
 properties {materialDomain=MD_LightFunction, pixelDepthOffsetMode=PDOM_Legacy, usedShadingModels=Default Lit}
 nodes [
  SkyAtmosphereLightDirection()
  n1 = ComponentMask(None=CollectionParameter() {collection=/Volumetrics/Content/PlanetPrototype/Materials/MPC_Planet, parameterName=SunVector}) {b=true, g=true, r=true}
  n2 = Divide(A=ComponentMask(None=@n1) {g=true, r=true}, B=ComponentMask(None=@n1) {b=true})
  n3 = ScalarParameter() {defaultValue=0.1, parameterName=Z}
  n4 = ComponentMask(None=WorldPosition().XYZ) {g=true, r=true}
  n5 = ScalarParameter() {defaultValue=3550000, parameterName=fasdf}
  TextureSample(UVs=Add(A=Multiply(A=@n2, B=@n3), B=Divide(A=@n4, B=@n5))) {texture=/Engine/EngineVolumetrics/Fogsheet/Materials/T_EV_FogSheet_CloudTex_02}
 ]
}
