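# SampleCloudScatteredLight is a material function that supplies tinted sky atmosphere scattered light, named for use in cloud shading.
# It multiplies a Color Multiply input, which falls back to white when unconnected, by the SkyAtmosphereDistantLightScatteredLuminance node and outputs the product.
# The light comes straight from the Sky Atmosphere's distant scattered luminance, with no other processing in the function.
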MaterialFunction /Volumetrics/Content/Sky/Functions/SampleCloudScatteredLight
graph {nodes [FunctionOutput(None=Multiply(
    A=FunctionInput() {bUsePreviewValueAsDefault=true, inputName=Color Multiply, previewValue=<1 1 1 1>}
    B=SkyAtmosphereDistantLightScatteredLuminance()
   ))]}
