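# SampleCloudSunLight is a material function that computes tinted, view-dependent sun light, by its name for sky clouds.
# A SchlickPhase call takes the dot product of sky atmosphere light direction and camera vector, plus an Anisotropy input, for directional scattering.
# Sky atmosphere light illuminance is sampled at a world point on the origin's vertical axis at the Altitude input, ignoring pixel position.
# The output multiplies phase, illuminance and a Color Multiply input.
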MaterialFunction /Volumetrics/Content/Sky/Functions/SampleCloudSunLight
graph {nodes [
  n1 = FunctionInput() {bUsePreviewValueAsDefault=true, inputName=Anisotropy, inputType=FunctionInput_Scalar, sortPriority=1, previewValue=<0.3 1 1 1>}
  n2 = MaterialFunctionCall!(Cosine=DotProduct(A=SkyAtmosphereLightDirection(), B=CameraVectorWS()), Anisotropy=@n1) {materialFunction=/Engine/Functions/Engine_MaterialFunctions01/Volumetrics/SchlickPhase}
  n3 = Constant()
  n4 = FunctionInput() {bUsePreviewValueAsDefault=true, inputName=Altitude, inputType=FunctionInput_Scalar, sortPriority=2, previewValue=<20000 1 1 1>}
  n5 = SkyAtmosphereLightIlluminance!(World Position=AppendVector(A=AppendVector(A=@n3, B=@n3), B=@n4))
  FunctionOutput(None=Multiply(A=FunctionInput() {bUsePreviewValueAsDefault=true, inputName=Color Multiply, previewValue=<1 1 1 1>}, B=Multiply(A=@n2.Result, B=@n5)))
 ]}
