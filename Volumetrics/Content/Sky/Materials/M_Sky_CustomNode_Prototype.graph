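# Prototype unlit, alpha-composite sky material with depth test disabled that renders a volumetric cloud layer around a planet-scale sphere.
# A Custom node calls CloudScene.Render from VolumetricCloudFunctions.ush, fed by a cloud mask, volume noise texture and packed parameter vectors.
# Material functions supply sun and scattered light, time pans the detail, and temporal-AA dither varies step size.
# The float4 result splits into color and inverted alpha.
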Material /Volumetrics/Content/Sky/Materials/M_Sky_CustomNode_Prototype
graph {
 expressions {
  MaterialExpressionAdd_0 {
   class=MaterialExpressionAdd
   inputs [
    A {from=MaterialExpressionMultiply_4}
    B {from=""}
   ]
  }
  MaterialExpressionAppendVector_1 {
   class=MaterialExpressionAppendVector
   inputs [
    A {from=MaterialExpressionVectorParameter_3, output=""}
    B {from=MaterialExpressionVectorParameter_3, output=""}
   ]
  }
  MaterialExpressionAppendVector_2 {
   class=MaterialExpressionAppendVector
   inputs [
    A {from=MaterialExpressionVectorParameter_10, output=""}
    B {from=MaterialExpressionVectorParameter_10, output=""}
   ]
  }
  MaterialExpressionAppendVector_6 {
   class=MaterialExpressionAppendVector
   inputs [
    A {from=MaterialExpressionVectorParameter_13, output=""}
    B {from=MaterialExpressionVectorParameter_13, output=""}
   ]
  }
  MaterialExpressionAppendVector_7 {
   class=MaterialExpressionAppendVector
   inputs [
    A {from=MaterialExpressionVectorParameter_6, output=""}
    B {from=MaterialExpressionVectorParameter_6, output=""}
   ]
  }
  MaterialExpressionAppendVector_8 {
   class=MaterialExpressionAppendVector
   inputs [
    A {from=MaterialExpressionVectorParameter_5, output=""}
    B {from=MaterialExpressionVectorParameter_5, output=""}
   ]
  }
  MaterialExpressionAppendVector_9 {
   class=MaterialExpressionAppendVector
   inputs [
    A {from=MaterialExpressionVectorParameter_7, output=""}
    B {from=MaterialExpressionVectorParameter_7, output=""}
   ]
  }
  MaterialExpressionComponentMask_2 {class=MaterialExpressionComponentMask, inputs [None {from=MaterialExpressionCustom_0}], properties {b=true, g=true, r=true}}
  MaterialExpressionComponentMask_3 {class=MaterialExpressionComponentMask, inputs [None {from=MaterialExpressionCustom_0}], properties {a=true}}
  MaterialExpressionConstant3Vector_1 {class=MaterialExpressionConstant3Vector, properties {constant=<rgba 1 0 0 0>}}
  MaterialExpressionConstant_0 {class=MaterialExpressionConstant}
  MaterialExpressionCustom_0 {
   class=MaterialExpressionCustom
   inputs [
    PlanetRadius {from=MaterialExpressionMultiply_2, output=""}
    MaxSteps {from=MaterialExpressionScalarParameter_7, output=""}
    StepSize {from=MaterialExpressionReroute_1, output=""}
    CloudMask {from=MaterialExpressionTextureObjectParameter_1, output=""}
    CloudMaskParams {from=MaterialExpressionAppendVector_1, output=""}
    None {from="", output=""}
    None {from="", output=""}
    None {from="", output=""}
    FalloffParams {from=MaterialExpressionAppendVector_7, output=""}
    None {from="", output=""}
    None {from="", output=""}
    None {from="", output=""}
    VolumeTex {from=MaterialExpressionTextureObjectParameter_3, output=""}
    Detail1Params {from=MaterialExpressionAppendVector_8, output=""}
    Detail2Params {from=MaterialExpressionAppendVector_9, output=""}
    None {from="", output=""}
    None {from="", output=""}
    None {from="", output=""}
    SunColor {from=MaterialExpressionMaterialFunctionCall_1, output=Result}
    SkyColor {from=MaterialExpressionMaterialFunctionCall_2, output=Result}
    GroundColor {from=MaterialExpressionMultiply_3, output=""}
    None {from="", output=""}
    None {from="", output=""}
    None {from="", output=""}
    ScatteringParams {from=MaterialExpressionAppendVector_2, output=""}
    Pan {from=MaterialExpressionMultiply_7, output=""}
    DensityParams {from=MaterialExpressionAppendVector_6, output=""}
    testoffset {from=MaterialExpressionScalarParameter_5, output=""}
   ]
   properties {
    code="##include \"/Plugin/Volumetrics/Private/VolumetricCloudFunctions.ush\"\r\n\r\n//return 0; \r\n\r\nCloudScene CloudLayer;\r\nreturn CloudLayer.Render();"
    description=SphereIntersection
    outputType=CMOT_Float4
    inputs [
     {
      inputName=PlanetRadius
      input {
       expression=&self:MaterialExpressionMultiply_2
       inputName=None
       mask=0
       maskA=0
       maskB=0
       maskG=0
       maskR=0
       outputIndex=0
      }
     }
     {
      inputName=MaxSteps
      input {
       expression=&self:MaterialExpressionScalarParameter_7
       inputName=None
       mask=0
       maskA=0
       maskB=0
       maskG=0
       maskR=0
       outputIndex=0
      }
     }
     {
      inputName=StepSize
      input {
       expression=&self:MaterialExpressionReroute_1
       inputName=None
       mask=0
       maskA=0
       maskB=0
       maskG=0
       maskR=0
       outputIndex=0
      }
     }
     {
      inputName=CloudMask
      input {
       expression=&self:MaterialExpressionTextureObjectParameter_1
       inputName=None
       mask=0
       maskA=0
       maskB=0
       maskG=0
       maskR=0
       outputIndex=0
      }
     }
     {
      inputName=CloudMaskParams
      input {
       expression=&self:MaterialExpressionAppendVector_1
       inputName=None
       mask=0
       maskA=0
       maskB=0
       maskG=0
       maskR=0
       outputIndex=0
      }
     }
     {inputName=None, input {expression=null, inputName=None, mask=0, maskA=0, maskB=0, maskG=0, maskR=0, outputIndex=0}}
     {inputName=None, input {expression=null, inputName=None, mask=0, maskA=0, maskB=0, maskG=0, maskR=0, outputIndex=0}}
     {inputName=None, input {expression=null, inputName=None, mask=0, maskA=0, maskB=0, maskG=0, maskR=0, outputIndex=0}}
     {
      inputName=FalloffParams
      input {
       expression=&self:MaterialExpressionAppendVector_7
       inputName=None
       mask=0
       maskA=0
       maskB=0
       maskG=0
       maskR=0
       outputIndex=0
      }
     }
     {inputName=None, input {expression=null, inputName=None, mask=0, maskA=0, maskB=0, maskG=0, maskR=0, outputIndex=0}}
     {inputName=None, input {expression=null, inputName=None, mask=0, maskA=0, maskB=0, maskG=0, maskR=0, outputIndex=0}}
     {inputName=None, input {expression=null, inputName=None, mask=0, maskA=0, maskB=0, maskG=0, maskR=0, outputIndex=0}}
     {
      inputName=VolumeTex
      input {
       expression=&self:MaterialExpressionTextureObjectParameter_3
       inputName=None
       mask=0
       maskA=0
       maskB=0
       maskG=0
       maskR=0
       outputIndex=0
      }
     }
     {
      inputName=Detail1Params
      input {
       expression=&self:MaterialExpressionAppendVector_8
       inputName=None
       mask=0
       maskA=0
       maskB=0
       maskG=0
       maskR=0
       outputIndex=0
      }
     }
     {
      inputName=Detail2Params
      input {
       expression=&self:MaterialExpressionAppendVector_9
       inputName=None
       mask=0
       maskA=0
       maskB=0
       maskG=0
       maskR=0
       outputIndex=0
      }
     }
     {inputName=None, input {expression=null, inputName=None, mask=1, maskA=0, maskB=1, maskG=1, maskR=1, outputIndex=0}}
     {inputName=None, input {expression=null, inputName=None, mask=0, maskA=0, maskB=0, maskG=0, maskR=0, outputIndex=0}}
     {inputName=None, input {expression=null, inputName=None, mask=0, maskA=0, maskB=0, maskG=0, maskR=0, outputIndex=0}}
     {
      inputName=SunColor
      input {
       expression=&self:MaterialExpressionMaterialFunctionCall_1
       inputName=None
       mask=0
       maskA=0
       maskB=0
       maskG=0
       maskR=0
       outputIndex=0
      }
     }
     {
      inputName=SkyColor
      input {
       expression=&self:MaterialExpressionMaterialFunctionCall_2
       inputName=None
       mask=0
       maskA=0
       maskB=0
       maskG=0
       maskR=0
       outputIndex=0
      }
     }
     {
      inputName=GroundColor
      input {
       expression=&self:MaterialExpressionMultiply_3
       inputName=None
       mask=0
       maskA=0
       maskB=0
       maskG=0
       maskR=0
       outputIndex=0
      }
     }
     {inputName=None, input {expression=null, inputName=None, mask=0, maskA=0, maskB=0, maskG=0, maskR=0, outputIndex=0}}
     {inputName=None, input {expression=null, inputName=None, mask=0, maskA=0, maskB=0, maskG=0, maskR=0, outputIndex=0}}
     {inputName=None, input {expression=null, inputName=None, mask=0, maskA=0, maskB=0, maskG=0, maskR=0, outputIndex=0}}
     {
      inputName=ScatteringParams
      input {
       expression=&self:MaterialExpressionAppendVector_2
       inputName=None
       mask=0
       maskA=0
       maskB=0
       maskG=0
       maskR=0
       outputIndex=0
      }
     }
     {
      inputName=Pan
      input {
       expression=&self:MaterialExpressionMultiply_7
       inputName=None
       mask=0
       maskA=0
       maskB=0
       maskG=0
       maskR=0
       outputIndex=0
      }
     }
     {
      inputName=DensityParams
      input {
       expression=&self:MaterialExpressionAppendVector_6
       inputName=None
       mask=0
       maskA=0
       maskB=0
       maskG=0
       maskR=0
       outputIndex=0
      }
     }
     {
      inputName=testoffset
      input {
       expression=&self:MaterialExpressionScalarParameter_5
       inputName=None
       mask=0
       maskA=0
       maskB=0
       maskG=0
       maskR=0
       outputIndex=0
      }
     }
    ]
   }
  }
  MaterialExpressionMaterialFunctionCall_0 {
   class=MaterialExpressionMaterialFunctionCall
   inputs [
    Alpha Threshold {from=""}
    Random {from=""}
   ]
   properties {materialFunction=&DitherTemporalAA}
  }
  MaterialExpressionMaterialFunctionCall_1 {
   class=MaterialExpressionMaterialFunctionCall
   inputs [
    Color Multiply {from=MaterialExpressionScalarParameter_12}
    Anisotropy {from=MaterialExpressionScalarParameter_15}
    Altitude {from=MaterialExpressionScalarParameter_16}
   ]
   properties {materialFunction=/Volumetrics/Content/Sky/Functions/SampleCloudSunLight}
  }
  MaterialExpressionMaterialFunctionCall_2 {
   class=MaterialExpressionMaterialFunctionCall
   inputs [Color Multiply {from=MaterialExpressionScalarParameter_20}]
   properties {materialFunction=/Volumetrics/Content/Sky/Functions/SampleCloudScatteredLight}
  }
  MaterialExpressionMaterialFunctionCall_3 {
   class=MaterialExpressionMaterialFunctionCall
   inputs [
    Time {from=""}
    Speed {from=MaterialExpressionScalarParameter_13}
   ]
   properties {materialFunction=&TimeWithSpeedVariable}
  }
  MaterialExpressionMultiply_0 {
   class=MaterialExpressionMultiply
   inputs [
    A {from=MaterialExpressionScalarParameter_11}
    B {from=MaterialExpressionAdd_0}
   ]
  }
  MaterialExpressionMultiply_2 {
   class=MaterialExpressionMultiply
   inputs [
    A {from=MaterialExpressionScalarParameter_1}
    B {from=MaterialExpressionScalarParameter_6}
   ]
  }
  MaterialExpressionMultiply_3 {
   class=MaterialExpressionMultiply
   inputs [
    A {from=MaterialExpressionVectorParameter_8, output=RGB}
    B {from=MaterialExpressionVectorParameter_8, output=RGB}
   ]
  }
  MaterialExpressionMultiply_4 {
   class=MaterialExpressionMultiply
   inputs [
    A {from=MaterialExpressionMaterialFunctionCall_0, output=Result}
    B {from=MaterialExpressionScalarParameter_2}
   ]
  }
  MaterialExpressionMultiply_7 {
   class=MaterialExpressionMultiply
   inputs [
    A {from=MaterialExpressionMaterialFunctionCall_3, output=Frac Time}
    B {from=MaterialExpressionConstant3Vector_1}
   ]
  }
  MaterialExpressionOneMinus_0 {class=MaterialExpressionOneMinus, inputs [None {from=MaterialExpressionComponentMask_3}]}
  MaterialExpressionReroute_1 {class=MaterialExpressionReroute, inputs [None {from=MaterialExpressionMultiply_0}]}
  MaterialExpressionScalarParameter_1 {class=MaterialExpressionScalarParameter, properties {defaultValue=636000000, parameterName=Radius Meters}}
  MaterialExpressionScalarParameter_11 {class=MaterialExpressionScalarParameter, properties {defaultValue=15000, parameterName=StepSize}}
  MaterialExpressionScalarParameter_12 {class=MaterialExpressionScalarParameter, properties {defaultValue=1, parameterName=Brightness Boost}}
  MaterialExpressionScalarParameter_13 {class=MaterialExpressionScalarParameter, properties {defaultValue=0.04, parameterName=DetailPanSpeed}}
  MaterialExpressionScalarParameter_15 {class=MaterialExpressionScalarParameter, properties {defaultValue=0.3, parameterName=Anisotropy}}
  MaterialExpressionScalarParameter_16 {class=MaterialExpressionScalarParameter, properties {defaultValue=20000, parameterName=Layer1 Alt}}
  MaterialExpressionScalarParameter_2 {class=MaterialExpressionScalarParameter, properties {defaultValue=0.2, parameterName=Dither}}
  MaterialExpressionScalarParameter_20 {class=MaterialExpressionScalarParameter, properties {defaultValue=1, parameterName=AmbientBoost}}
  MaterialExpressionScalarParameter_5 {class=MaterialExpressionScalarParameter}
  MaterialExpressionScalarParameter_6 {class=MaterialExpressionScalarParameter, properties {defaultValue=1, parameterName=Earth Scale}}
  MaterialExpressionScalarParameter_7 {class=MaterialExpressionScalarParameter, properties {defaultValue=128, parameterName=MaxSteps}}
  MaterialExpressionTextureObjectParameter_1 {
   class=MaterialExpressionTextureObjectParameter
   inputs [
    UVs {from="", output=""}
    Apply View MipBias {from="", output=""}
   ]
   properties {
    parameterName=CloudMask
    samplerType=SAMPLERTYPE_LinearColor
    texture=/Engine/Functions/Engine_MaterialFunctions02/ExampleContent/Textures/OffsetNoiseDistanceFields_ForNormals
   }
  }
  MaterialExpressionTextureObjectParameter_3 {
   class=MaterialExpressionTextureObjectParameter
   inputs [
    UVs {from="", output=""}
    Apply View MipBias {from="", output=""}
   ]
   properties {parameterName=VolumeTexture, samplerType=SAMPLERTYPE_LinearColor, texture=/Volumetrics/Content/VolumeTextures/Textures/VT_Voronoi}
  }
  MaterialExpressionVectorParameter_10 {
   class=MaterialExpressionVectorParameter
   properties {
    parameterName=ScatteringParams
    channelNames {a=Scatter Shadow Density, b=Scatter Multiply, g=Scatter Mask Power, r=Volume Scatter Boost}
    defaultValue=<rgba 0 64 0 0.5>
   }
  }
  MaterialExpressionVectorParameter_13 {
   class=MaterialExpressionVectorParameter
   properties {
    parameterName=DensityParams
    channelNames {a=Powder Density, b=Ambient Density, g=Shadow Density, r=Cloud Density}
    defaultValue=<rgba 50000 0.5 1 1>
   }
  }
  MaterialExpressionVectorParameter_3 {
   class=MaterialExpressionVectorParameter
   properties {parameterName=CloudBaseParams, channelNames {a=Offset Y, b=Offset X, g=Cloud Bias, r=Cloud Scale}, defaultValue=<rgba 12050000 0.2 0 0.7>}
  }
  MaterialExpressionVectorParameter_5 {
   class=MaterialExpressionVectorParameter
   properties {
    parameterName=CloudDetailParams
    channelNames {a=Altitude Shift, b=Detail 1 Mip, g=Detail 1 Amount, r=Detail 1 Scale Km}
    defaultValue=<rgba 50 0 1 0>
   }
  }
  MaterialExpressionVectorParameter_6 {
   class=MaterialExpressionVectorParameter
   properties {
    parameterName=CloudFalloffParams
    channelNames {a=Lower Falloff, b=Upper Falloff, g=Layer Thickness, r=Layer Altitude}
    defaultValue=<rgba 400000 3200000 1 0.5>
   }
  }
  MaterialExpressionVectorParameter_7 {
   class=MaterialExpressionVectorParameter
   properties {parameterName=CloudDetail2Params, channelNames {a="", b=Detail 2 Mip, g=Detail 2 Amount, r=Detail 2 Scale Km}}
  }
  MaterialExpressionVectorParameter_8 {class=MaterialExpressionVectorParameter, properties {parameterName=GroundColor, defaultValue=<rgba 0.5833 0.3222 0.1613 1>}}
 }
 properties {
  bDisableDepthTest=true
  bUseTranslucencyVertexFog=false
  blendMode=BLEND_AlphaComposite
  pixelDepthOffsetMode=PDOM_Legacy
  shadingModel=MSM_Unlit
  usedShadingModels=Unlit
 }
}
