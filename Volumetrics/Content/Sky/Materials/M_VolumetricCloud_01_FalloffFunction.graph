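# Volumetric cloud material that builds cloud density from a world-projected render-target coverage mask and an altitude falloff function.
# Falloff pairs an SCurve upper fade with an exponential lower fade, and its biased reciprocal erodes the mask.
# Two scales of Voronoi volume noise add detail before density and altitude-weighted tint scaling.
# An advanced output sets dual-phase, multi-scattering lighting; the density chain and an alternative exponential falloff feed no shown output.
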Material /Volumetrics/Content/Sky/Materials/M_VolumetricCloud_01_FalloffFunction
graph {
 properties {
  bUsedWithVolumetricCloud=true
  blendMode=BLEND_Additive
  materialDomain=MD_Volume
  pixelDepthOffsetMode=PDOM_Legacy
  usedShadingModels=Default Lit
 }
 nodes [
  n1 = Divide(A=WorldPosition().XYZ, B=Multiply(A=ScalarParameter() {defaultValue=400, parameterName=Mask Scale}) {constB=100000})
  n2 = VectorParameter() {parameterName=Offset, defaultValue=<rgba 0.5 0.5 0 0>}
  n3 = TextureSample!(UVs=ComponentMask(None=Subtract(A=@n1, B=@n2.RGB)) {g=true, r=true}, Tex, Level, Apply View MipBias) {
   constMipValue=0
   mipValueMode=TMVM_MipLevel
   samplerType=SAMPLERTYPE_LinearColor
   texture=/Volumetrics/Tools/CloudCompositing/RenderTargets/RT_CloudMask_01
  }
  n4 = CloudSampleAttribute()
  n5 = MaterialFunctionCall(In=@n4.NormAltitudeInLayer, Power=ScalarParameter() {defaultValue=0.5, parameterName=Upper Falloff}) {materialFunction=&SCurve}
  MaterialExpressionMultiply_33 = Multiply(A=ScalarParameter() {defaultValue=64, parameterName=Lower Falloff}, B=Multiply(A=@n4.NormAltitudeInLayer))
  n6 = Custom!(x=@MaterialExpressionMultiply_33) {
   code="return exp(-x);"
   description="exp(-x)"
   outputType=CMOT_Float1
   inputs [{inputName=x, input {expression=&self:MaterialExpressionMultiply_33}}]
  }
  n7 = Max(A=Multiply(A=OneMinus(None=@n5.Result), B=OneMinus(None=@n6)) {desc="GetLayerFalloff()"}) {constB=0.01}
  n8 = Subtract(A=@n3.R, B=Multiply(A=Divide(B=@n7), B=ScalarParameter() {defaultValue=0.5, parameterName=Cloud Bias}))
  n9 = ScalarParameter() {defaultValue=64, parameterName=scale}
  n10 = TextureSample(UVs=Divide(A=WorldPosition().XYZ, B=Multiply(A=@n9) {constB=100000}), Level=ScalarParameter() {parameterName=Volume Mip}) {mipValueMode=TMVM_MipLevel, samplerType=SAMPLERTYPE_LinearColor, texture=/Volumetrics/Content/VolumeTextures/Textures/VT_Voronoi_Simple}
  n11 = ScalarParameter() {defaultValue=29, parameterName=scale2}
  n12 = TextureSample(UVs=Divide(A=WorldPosition().XYZ, B=Multiply(A=@n11) {constB=100000}), Level=ScalarParameter() {parameterName=Volume Mip}) {mipValueMode=TMVM_MipLevel, samplerType=SAMPLERTYPE_LinearColor, texture=&VT_Voronoi_Simple}
  n13 = Add(
   A=Multiply(A=Subtract(A=@n10.R) {constB=0.5}, B=ScalarParameter() {defaultValue=-0.4, parameterName=Detail})
   B=Multiply(A=Subtract(A=@n12.R) {constB=0.5}, B=ScalarParameter() {defaultValue=-0.2, parameterName=Detail2})
  )
  n14 = ScalarParameter() {defaultValue=24, parameterName=Cloud Density}
  n15 = VectorParameter() {parameterName=ecx, defaultValue=<rgba 0.5 0.6 0.7 0.5>}
  Multiply(
   A=Multiply(A=Add(A=@n8, B=@n13), B=Divide(A=@n14) {constB=100})
   B=Multiply(A=Multiply(A=@n15.RGB, B=@n15.RGB), B=CloudSampleAttribute().NormAltitudeInLayer)
  )
  n16 = OneMinus(None=@n4.NormAltitudeInLayer)
  MaterialExpressionMultiply_20 = Multiply(A=ScalarParameter() {defaultValue=0.5, parameterName=Upper Falloff}, B=@n16)
  n17 = Custom(x=@MaterialExpressionMultiply_20) {
   code="return exp(-x);"
   description="exp(-x)"
   outputType=CMOT_Float1
   inputs [{inputName=x, input {expression=&self:MaterialExpressionMultiply_20}}]
  }
  MaterialExpressionMultiply_19 = Multiply(A=ScalarParameter() {defaultValue=64, parameterName=Lower Falloff}, B=@n4.NormAltitudeInLayer)
  n18 = Custom(x=@MaterialExpressionMultiply_19) {
   code="return exp(-x);"
   description="exp(-x)"
   outputType=CMOT_Float1
   inputs [{inputName=x, input {expression=&self:MaterialExpressionMultiply_19}}]
  }
  Multiply(A=OneMinus(None=@n17), B=OneMinus(None=@n18)) {desc="GetLayerFalloff()"}
  Multiply(A=@n3.R, B=@n16)
  ScalarParameter() {defaultValue=0.5, parameterName=Cloud Bias}
  VectorParameter() {parameterName=Albedo, defaultValue=<rgba 1 1 1 0>}
  VolumetricAdvancedMaterialOutput!(
   PhaseG=ScalarParameter() {defaultValue=0.5, group=Scattering and Phase, parameterName=Phase A}
   PhaseG2=ScalarParameter() {defaultValue=0.8, group=Scattering and Phase, parameterName=Phase B}
   PhaseBlend=ScalarParameter() {defaultValue=0.4, group=Scattering and Phase, parameterName=Phasey Phasey}
   MultiScatteringContribution=ScalarParameter() {defaultValue=0.5, group=Scattering and Phase, parameterName=MS C}
   MultiScatteringOcclusion=ScalarParameter() {defaultValue=0.2, group=Scattering and Phase, parameterName=MS Occ}
   MultiScatteringEccentricity=ScalarParameter() {defaultValue=0.5, group=Scattering and Phase, parameterName=MS Ec}
   ConservativeDensity
  ) {multiScatteringApproximationOctaveCount=2}
 ]
}
