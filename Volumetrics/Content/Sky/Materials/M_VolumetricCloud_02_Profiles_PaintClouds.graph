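# Volumetric cloud material that shapes clouds from a 2D storm mask animated along the flow-map render target RT_Flowmap_01.
# The biased mask and layer altitude look up a profile texture for the base vertical shape.
# Two flow-mapped Voronoi octaves, driven by flow plus curl noise and the first distorting the second, add altitude-weighted detail.
# Soft 1-exp(-x) curves saturate density, and an advanced volumetric output sets dual-lobe phase and multi-scattering.
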Material /Volumetrics/Content/Sky/Materials/M_VolumetricCloud_02_Profiles_PaintClouds
graph {
 properties {
  bUsedWithVolumetricCloud=true
  blendMode=BLEND_Additive
  materialDomain=MD_Volume
  pixelDepthOffsetMode=PDOM_Legacy
  usedShadingModels=Default Lit
 }
 nodes [
  n1 = CloudSampleAttribute()
  n2 = CollectionParameter() {collection=/Volumetrics/Tools/CloudCompositing/MPC/MPC_Clouds, parameterName=SkyTextureScaleKm}
  n3 = Subtract(
   A=Divide(A=WorldPosition().XYZ, B=Multiply(A=@n2) {constB=100000})
   B=VectorParameter() {parameterName=Offset, defaultValue=<rgba 0.5 0.5 0 0>}.RGB
  )
  n4 = ComponentMask(None=@n3) {g=true, r=true}
  n5 = TextureSample!(UVs=@n4, Tex, Level, Apply View MipBias) {
   constMipValue=0
   mipValueMode=TMVM_MipLevel
   samplerType=SAMPLERTYPE_LinearColor
   texture=/Volumetrics/Tools/CloudCompositing/RenderTargets/RT_Flowmap_01
  }
  n6 = Multiply(A=ComponentMask(None=@n5.RGB) {g=true, r=true}, B=ScalarParameter() {defaultValue=-1, parameterName=Velocity Scale})
  n7 = MaterialFunctionCall(Speed=ScalarParameter() {defaultValue=0.1, parameterName=Speed}) {materialFunction=&TimeWithSpeedVariable}
  n8 = ComponentMask(None=MaterialFunctionCall!(
    Volume Texture=TextureObject() {samplerType=SAMPLERTYPE_LinearColor, texture=/Volumetrics/Tools/CloudCompositing/Textures/T_CloudMask_01_Storm}
    UV=@n4
    Velocity=@n6
    Override Time=@n7.Time
    Mip Level
   ) {materialFunction=/Engine/Functions/Engine_MaterialFunctions02/Texturing/Flowmaps_2D}.Result RGB) {r=true}
  n9 = ScalarParameter() {defaultValue=0.4, parameterName=Bias}
  MaterialExpressionMax_0 = Max(A=Multiply(A=Subtract(A=@n8, B=@n9), B=ScalarParameter() {defaultValue=2, parameterName=Density})) {constB=0}
  n10 = Custom!(x=@MaterialExpressionMax_0) {
   code="return 1-exp(-x);"
   outputType=CMOT_Float1
   inputs [{inputName=x, input {expression=&self:MaterialExpressionMax_0}}]
  }
  n11 = Constant() {r=0.0059}
  n12 = TextureSample(UVs=Clamp(None=AppendVector(A=@n1.NormAltitudeInLayer, B=@n10), Min=@n11, Max=OneMinus(None=@n11)), Level=Constant()) {mipValueMode=TMVM_MipLevel, samplerType=SAMPLERTYPE_LinearColor, texture=/Volumetrics/Tools/CloudCompositing/Textures/Profile_01}
  n13 = LinearInterpolate(A=ScalarParameter() {defaultValue=0.3, parameterName=Bottom Noise}, Alpha=CloudSampleAttribute().NormAltitudeInLayer)
  n14 = TextureObject() {samplerType=SAMPLERTYPE_LinearColor, texture=/Volumetrics/Content/VolumeTextures/Textures/VT_Voronoi}
  n15 = Divide(A=WorldPosition().XYZ, B=Multiply(A=ScalarParameter() {defaultValue=64, parameterName=scale}) {constB=100000})
  n16 = Constant()
  n17 = MaterialFunctionCall(Vector 2=@n5.RGB) {materialFunction=/Engine/Functions/Engine_MaterialFunctions02/Utility/VectorLength}
  n18 = ScalarParameter() {defaultValue=0.1, parameterName=Velocity for Curl}
  n19 = Multiply(A=Saturate(None=Divide(A=@n17.V2 Length, B=@n18)), B=ScalarParameter() {defaultValue=0.001, parameterName=Curl Amount})
  n20 = TextureSample(UVs=Multiply(A=@n3, B=ScalarParameter() {defaultValue=8, parameterName=TilNoiseTiling})) {
   constMipValue=0
   mipValueMode=TMVM_MipLevel
   samplerType=SAMPLERTYPE_LinearColor
   texture=/Volumetrics/Content/VolumeTextures/Textures/VT_CurlNoise
  }
  n21 = Multiply(
   A=Add(A=AppendVector(A=@n6, B=@n16), B=Multiply(A=@n19, B=@n20.RGB))
   B=ScalarParameter() {defaultValue=32, parameterName=Noise Flowmaps Speed}
  )
  n22 = ScalarParameter() {defaultValue=2, parameterName=Detail Mip}
  n23 = Subtract(A=ComponentMask(None=MaterialFunctionCall!(Volume Texture=@n14, UVW=@n15, Velocity=@n21, Override Time=@n7.Time, Mip Level=@n22) {materialFunction=/Engine/Functions/Engine_MaterialFunctions02/Texturing/Flowmaps_3D}.Result RGB) {r=true}) {constB=0.5}
  n24 = Multiply(A=@n23, B=ScalarParameter() {defaultValue=0.2, parameterName=Noise UV Distortion})
  n25 = Divide(A=WorldPosition().XYZ, B=Multiply(A=ScalarParameter() {defaultValue=8, parameterName=scale2}) {constB=100000})
  n26 = ComponentMask(None=MaterialFunctionCall(
    Volume Texture=TextureObject() {samplerType=SAMPLERTYPE_LinearColor, texture=&VT_Voronoi}
    UVW=Add(A=@n24, B=@n25)
    Velocity=@n21
    Override Time=@n7.Time
    Mip Level=@n22
   ) {materialFunction=&Flowmaps_3D}.Result RGB) {r=true}
  n27 = Add(
   A=Multiply(A=@n23, B=ScalarParameter() {defaultValue=0.4, parameterName=Detail})
   B=Multiply(A=Subtract(A=@n26) {constB=0.5}, B=ScalarParameter() {defaultValue=0.31, parameterName=Detail2})
  )
  MaterialExpressionMultiply_15 = Multiply(A=Add(A=@n12.R, B=Multiply(A=@n13, B=@n27)), B=ScalarParameter() {defaultValue=16, parameterName=Pre Exponential Density})
  n28 = Custom(x=@MaterialExpressionMultiply_15) {
   code="return 1-exp(-x);"
   outputType=CMOT_Float1
   inputs [{inputName=x, input {expression=&self:MaterialExpressionMultiply_15}}]
  }
  LinearInterpolate(Alpha=@n28)
  n29 = ScalarParameter() {defaultValue=250, parameterName=3}
  n30 = ComponentMask(None=Divide(A=WorldPosition().XYZ, B=Multiply(A=@n29) {constB=100000})) {g=true, r=true}
  n31 = ScalarParameter() {parameterName=2D Mask Mip2}
  n32 = LinearInterpolate(
   A=ScalarParameter() {defaultValue=0.0001, parameterName=OutD}
   B=ScalarParameter() {defaultValue=0.1, parameterName=OutD2}
   Alpha=Power(
    Base=TextureSample(UVs=@n30, Level=@n31) {mipValueMode=TMVM_MipLevel, texture=/Engine/MapTemplates/TilingNoise05}.R
    Exp=ScalarParameter() {defaultValue=2.5, parameterName=Param_1}
   )
  )
  n33 = VectorParameter() {parameterName=ecx, defaultValue=<rgba 0.5 0.575 0.65 0.5>}
  n34 = Multiply(A=@n33.RGB, B=@n33.RGB)
  Multiply(A=Multiply(A=@n28, B=Max(A=@n32) {constB=0}), B=@n34)
  Multiply(A=@n34, B=CloudSampleAttribute().NormAltitudeInLayer)
  MaterialExpressionMultiply_19 = Multiply(
   A=ScalarParameter() {defaultValue=1.5, parameterName=Depth Density}
   B=Add(A=@MaterialExpressionMultiply_15, B=ScalarParameter() {defaultValue=0.1, parameterName=Depth Density Bias})
  )
  n35 = Custom(x=@MaterialExpressionMultiply_19) {
   code="return 1-exp(-x);"
   outputType=CMOT_Float1
   inputs [{inputName=x, input {expression=&self:MaterialExpressionMultiply_19}}]
  }
  Multiply(A=@n35, B=VectorParameter() {parameterName=cc, defaultValue=<rgba 1 1 1 0>}.RGB)
  TextureSample(UVs=@n15, Level=ScalarParameter() {defaultValue=3, parameterName=Volume Mip}) {mipValueMode=TMVM_MipLevel, samplerType=SAMPLERTYPE_LinearColor, texture=&VT_Voronoi}
  TextureSample(UVs=@n4, Level=ScalarParameter() {parameterName=2D Mask Mip}) {mipValueMode=TMVM_MipLevel, samplerType=SAMPLERTYPE_LinearColor, texture=/Volumetrics/Tools/CloudCompositing/RenderTargets/RT_CloudMask_01}
  VolumetricAdvancedMaterialOutput!(
   PhaseG=ScalarParameter() {parameterName=Phase A}
   PhaseG2=ScalarParameter() {defaultValue=0.8, parameterName=Phase B}
   PhaseBlend=ScalarParameter() {defaultValue=0.5, parameterName=Phasey Phasey}
   MultiScatteringContribution=ScalarParameter() {defaultValue=0.6, parameterName=MS C}
   MultiScatteringOcclusion=Multiply(A=ScalarParameter() {defaultValue=0.5, parameterName=MS Occ}, B=ScalarParameter() {defaultValue=0.7, parameterName=Powertest})
   MultiScatteringEccentricity=ScalarParameter() {defaultValue=0.3, parameterName=MS Ec}
   ConservativeDensity
  ) {multiScatteringApproximationOctaveCount=2}
 ]
}
