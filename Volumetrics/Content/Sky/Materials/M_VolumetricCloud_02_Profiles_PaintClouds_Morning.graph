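# Volumetric cloud material for painted skies, a "Morning" variant with two cloud types and animated lightning.
# The RT_CloudMask_01 render target, advected by painted flowmap RT_Flowmap_01, indexes altitude profile textures to shape conservative density.
# Two chained 3D flowmapped Voronoi noises, pushed by curl noise and fake vertical advection, add detail.
# Exponential remaps of that density drive extinction and two-tone albedo, while scrolling VT_Lightning, masked by clouds and controlled through MPC_Clouds, supplies glow.
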Material /Volumetrics/Content/Sky/Materials/M_VolumetricCloud_02_Profiles_PaintClouds_Morning
graph {
 properties {
  bUsedWithVolumetricCloud=true
  blendMode=BLEND_Additive
  materialDomain=MD_Volume
  pixelDepthOffsetMode=PDOM_Legacy
  usedShadingModels=Default Lit
 }
 nodes [
  Constant() {r=0.05}
  MaterialExpressionMultiply_15 = Multiply(
   A=VolumetricAdvancedMaterialInput().ConservativeDensity as Float3
   B=ScalarParameter() {defaultValue=8, parameterName=Pre Exponential Density}
  )
  n1 = Custom!(x=@MaterialExpressionMultiply_15) {
   code="return 1-exp(-x);"
   description=exp
   outputType=CMOT_Float1
   inputs [{
     inputName=x
     input {expression=&self:MaterialExpressionMultiply_15}
    }]
  }
  n2 = ScalarParameter() {defaultValue=0.0001, parameterName=OutD}
  n3 = ScalarParameter() {defaultValue=0.002, parameterName=Dark Min Density}
  n4 = CollectionParameter() {collection=/Volumetrics/Tools/CloudCompositing/MPC/MPC_Clouds, parameterName=SkyTextureScaleKm}
  n5 = Subtract(
   A=Divide(A=WorldPosition().XYZ, B=MaterialFunctionCall!(Km=@n4) {materialFunction=/Engine/Functions/Engine_MaterialFunctions01/Units/Km-to-Cm}.Cm)
   B=VectorParameter() {parameterName=Offset, defaultValue=<rgba 0.5 0.5 0 0>}.RGB
  )
  n6 = ScalarParameter() {defaultValue=1, parameterName=Top Flow}
  n7 = CloudSampleAttribute()
  n8 = TextureSample!(UVs=ComponentMask(None=@n5) {g=true, r=true}, Tex, Level, Apply View MipBias) {
   constMipValue=0
   mipValueMode=TMVM_MipLevel
   samplerSource=SSM_Wrap_WorldGroupSettings
   samplerType=SAMPLERTYPE_LinearColor
   texture=/Volumetrics/Tools/CloudCompositing/RenderTargets/RT_Flowmap_01
  }
  n9 = ComponentMask(None=@n8.RGB) {g=true, r=true}
  n10 = Multiply(
   A=Multiply(A=LinearInterpolate(B=@n6, Alpha=@n7.NormAltitudeInLayer) {constA=1}, B=@n9)
   B=ScalarParameter() {defaultValue=-1, parameterName=Velocity Scale}
  )
  n11 = MaterialFunctionCall(Speed=ScalarParameter() {defaultValue=0.123, group=CloudBase, parameterName=Speed}) {materialFunction=&TimeWithSpeedVariable}
  n12 = MaterialFunctionCall!(
   Volume Texture=TextureObject() {samplerType=SAMPLERTYPE_LinearColor, texture=/Volumetrics/Tools/CloudCompositing/RenderTargets/RT_CloudMask_01}
   UV=ComponentMask(None=@n5) {g=true, r=true}
   Velocity=@n10
   Override Time=@n11.Time
   Mip Level
  ) {materialFunction=/Engine/Functions/Engine_MaterialFunctions02/Texturing/Flowmaps_2D}
  n13 = ComponentMask(None=@n12.Result RGB) {b=true, g=true, r=true}
  n14 = ScalarParameter() {defaultValue=0.4, group=CloudBase, parameterName=Bias}
  MaterialExpressionMax_0 = Max(A=Multiply(A=Subtract(A=@n13, B=@n14), B=ScalarParameter() {defaultValue=1, group=CloudBase, parameterName=Density})) {constB=0}
  n15 = Custom(x=@MaterialExpressionMax_0) {
   code="return 1-exp(-x);"
   inputs [{inputName=x, input {expression=&self:MaterialExpressionMax_0}}]
  }
  n16 = ComponentMask(None=@n15) {g=true}
  n17 = Divide(A=WorldPosition().XYZ, B=Multiply(A=ScalarParameter() {defaultValue=250, parameterName=3}) {constB=100000})
  n18 = Multiply(
   A=TextureSample(UVs=ComponentMask(None=@n17) {g=true, r=true}, Level=ScalarParameter() {parameterName=2D Mask Mip2}) {mipValueMode=TMVM_MipLevel, samplerSource=SSM_Wrap_WorldGroupSettings, texture=/Engine/MapTemplates/TilingNoise05}.R
   B=OneMinus(None=ComponentMask(None=@n12.Result RGB) {b=true})
  )
  n19 = ScalarParameter() {defaultValue=3, parameterName=Param_1}
  n20 = Max(A=LinearInterpolate(A=LinearInterpolate(A=@n2, B=@n3, Alpha=@n16), B=ScalarParameter() {defaultValue=0.2, parameterName=OutD2}, Alpha=Power(Base=@n18, Exp=@n19))) {constB=0}
  n21 = Multiply(A=@n20, B=@n1)
  n22 = VectorParameter() {parameterName=ecx, defaultValue=<rgba 1 1 1 1>}
  Multiply(A=@n21, B=Multiply(A=@n22.RGB, B=@n22.RGB))
  n23 = ScalarParameter() {defaultValue=0.5, parameterName=Darkness power}
  MaterialExpressionMultiply_19 = Multiply(
   A=Add(A=@MaterialExpressionMultiply_15, B=ScalarParameter() {defaultValue=64, parameterName=Depth Density Bias})
   B=ScalarParameter() {defaultValue=0.75, parameterName=Depth Density}
  )
  n24 = Custom(x=@MaterialExpressionMultiply_19) {
   code="return 1-exp(-x);"
   description=exp
   outputType=CMOT_Float1
   inputs [{
     inputName=x
     input {expression=&self:MaterialExpressionMultiply_19}
    }]
  }
  Multiply(
   A=LinearInterpolate(
    A=VectorParameter() {parameterName=Albedo, defaultValue=<rgba 1 1 1 0>}.RGB
    B=VectorParameter() {parameterName=2nd Cloud Albedo, defaultValue=<rgba 0.2 0.25 0.4 0>}.RGB
    Alpha=Power(Base=@n16, Exp=@n23)
   )
   B=@n24
  )
  n25 = WorldPosition()
  n26 = MaterialFunctionCall(Km=ScalarParameter() {defaultValue=200, parameterName=Lighting Pattern Scale}) {materialFunction=&Km-to-Cm}
  n27 = CollectionParameter() {collection=&MPC_Clouds, parameterName=LightningSpeed}
  n28 = AppendVector(
   A=ComponentMask(None=Divide(A=@n25.XYZ, B=@n26.Cm)) {g=true, r=true}
   B=LinearInterpolate(
    A=MaterialFunctionCall(Speed=@n27) {materialFunction=&TimeWithSpeedVariable}.Time
    B=CollectionParameter() {collection=&MPC_Clouds, parameterName=ManualTime}
    Alpha=CollectionParameter() {collection=&MPC_Clouds, parameterName=FreezeTime}
   )
  )
  n29 = TextureSample!(UVs=@n28, Tex, Apply View MipBias) {samplerSource=SSM_Wrap_WorldGroupSettings, texture=/Volumetrics/Content/VolumeTextures/Textures/VT_Lightning}
  n30 = Power(Base=@n29.R, Exp=ScalarParameter() {defaultValue=3, parameterName=Lightning Power})
  n31 = Multiply(A=Multiply(A=@n29.R, B=@n29.R), B=ScalarParameter() {parameterName=Fill Lightning Scatter})
  n32 = TextureSample!(UVs=@n28, Tex, Level=ScalarParameter() {defaultValue=2.5, parameterName=Lightning Mip}, Apply View MipBias) {mipValueMode=TMVM_MipLevel, samplerSource=SSM_Wrap_WorldGroupSettings, texture=&VT_Lightning}
  n33 = ScalarParameter() {defaultValue=0.05, parameterName=Lightning Spread}
  n34 = Constant3Vector() {constant=<rgba 1 2 0 0>}
  n35 = Multiply(
   A=Subtract(A=DotProduct(A=@n12.Result RGB, B=@n34), B=ScalarParameter() {defaultValue=1, parameterName=Light Mask Bias})
   B=ScalarParameter() {defaultValue=0.05, parameterName=Light Mask Intensity}
  )
  n36 = VolumetricAdvancedMaterialInput()
  n37 = Multiply(A=CloudSampleAttribute().NormAltitudeInLayer, B=ScalarParameter() {defaultValue=2, group=CloudBase, parameterName=Cloud 1 Height})
  n38 = ComponentMask(None=@n15) {r=true}
  n39 = Constant() {r=0.0059}
  n40 = TextureSample(UVs=Clamp(None=AppendVector(A=@n37, B=@n38), Min=@n39, Max=OneMinus(None=@n39))) {
   constMipValue=0
   mipValueMode=TMVM_MipLevel
   samplerSource=SSM_Wrap_WorldGroupSettings
   samplerType=SAMPLERTYPE_LinearColor
   texture=/Volumetrics/Tools/CloudCompositing/Textures/Profile_01
  }
  n41 = Subtract(A=@n40.R, B=Max(A=Subtract(A=@n37)) {constB=0})
  n42 = CloudSampleAttribute()
  n43 = Constant() {r=0.0059}
  n44 = TextureSample(UVs=Clamp(None=AppendVector(A=@n42.NormAltitudeInLayer, B=@n16), Min=@n43, Max=OneMinus(None=@n43))) {
   constMipValue=0
   mipValueMode=TMVM_MipLevel
   samplerSource=SSM_Wrap_WorldGroupSettings
   samplerType=SAMPLERTYPE_LinearColor
   texture=/Volumetrics/Tools/CloudCompositing/Textures/RT_Profiles
  }
  n45 = VectorParameter() {group=CloudBase, parameterName=CurveChannel, defaultValue=<rgba 0 0 1 0>}
  n46 = Subtract(A=DotProduct(A=@n44.RGBA, B=AppendVector(A=@n45.RGB, B=@n45.RGB))) {constB=0.5}
  n47 = ScalarParameter() {defaultValue=3, group=CloudBase, parameterName=Cloud 2 Density}
  n48 = LinearInterpolate(A=@n41, B=Max(A=Multiply(A=@n46, B=@n47), B=@n41), Alpha=ScalarParameter() {defaultValue=1, group=CloudBase, parameterName=Show 2nd Cloud})
  n49 = ScalarParameter() {defaultValue=0.5, parameterName=ScatteringTexAmount}
  n50 = Multiply(
   A=Add(A=Add(A=@n30, B=@n31), B=Multiply(A=@n32.R, B=@n33))
   B=Multiply(A=Saturate(None=@n35), B=LinearInterpolate(A=@n36.ConservativeDensity as Float3, B=@n48, Alpha=@n49))
  ) {constA=1}
  n51 = TextureObjectParameter() {parameterName=VolumeTexture, samplerType=SAMPLERTYPE_LinearColor, texture=/Volumetrics/Content/VolumeTextures/Textures/VT_Voronoi}
  n52 = WorldPosition()
  n53 = Multiply(A=ScalarParameter() {defaultValue=7, group=Flowmap1, parameterName=scale}) {constB=100000}
  n54 = OneMinus(None=CloudSampleAttribute().NormAltitudeInLayer)
  n55 = ScalarParameter() {defaultValue=0.1, group=StormFakeAdvection, parameterName=Storm Fake Advection Bias}
  MaterialExpressionMax_4 = Max(A=Multiply(A=Subtract(A=@n54, B=@n55), B=ScalarParameter() {defaultValue=4, group=StormFakeAdvection, parameterName=Storm Fake Advection Density})) {constB=0}
  n56 = Custom(x=@MaterialExpressionMax_4) {
   code="return 1-exp(-x);"
   description=exp
   outputType=CMOT_Float1
   inputs [{inputName=x, input {expression=&self:MaterialExpressionMax_4}}]
  }
  n57 = Multiply(A=@n56, B=VectorParameter() {group=StormFakeAdvection, parameterName=Fake Advection Dir, defaultValue=<rgba 0 0 1 0>}.RGB)
  n58 = Constant()
  n59 = MaterialFunctionCall(Vector 2=@n9) {materialFunction=/Engine/Functions/Engine_MaterialFunctions02/Utility/VectorLength}
  n60 = ScalarParameter() {defaultValue=0.1, parameterName=Velocity for Curl}
  n61 = Multiply(A=Saturate(None=Divide(A=@n59.V2 Length, B=@n60)), B=ScalarParameter() {defaultValue=0.001, parameterName=Curl Amount})
  n62 = ScalarParameter() {defaultValue=4, parameterName=TilNoiseTiling}
  n63 = TextureSample(UVs=Multiply(A=Multiply(A=@n5, B=@n62), B=VectorParameter() {parameterName=NUC, defaultValue=<rgba 1 1 2.5 0>}.RGB)) {
   constMipValue=0
   mipValueMode=TMVM_MipLevel
   samplerType=SAMPLERTYPE_LinearColor
   texture=/Volumetrics/Content/VolumeTextures/Textures/VT_CurlNoise
  }
  n64 = Multiply(
   A=Add(A=AppendVector(A=@n10, B=@n58), B=Multiply(A=@n61, B=@n63.RGB))
   B=ScalarParameter() {defaultValue=64, group=Flowmap1, parameterName=Noise Flowmaps Speed}
  )
  n65 = ScalarParameter() {defaultValue=3, group=Flowmap1, parameterName=Detail Mip}
  n66 = VectorParameter() {group=Flowmap1, parameterName=Flowmap1Channel, defaultValue=<rgba 1 0 0 0>}
  n67 = DotProduct(
   A=MaterialFunctionCall!(Volume Texture=@n51, UVW=Divide(A=@n52.XYZ, B=@n53), Velocity=Add(A=@n57, B=@n64), Override Time=@n11.Time, Mip Level=@n65) {materialFunction=/Engine/Functions/Engine_MaterialFunctions02/Texturing/Flowmaps_3D}.Result RGBA
   B=AppendVector(A=@n66.RGB, B=@n66.RGB)
  )
  n68 = Add(A=@n67, B=ScalarParameter() {defaultValue=-0.5, group=Flowmap1, parameterName=Flowmap1Bias, sliderMax=1, sliderMin=-1})
  n69 = Multiply(A=@n68, B=ScalarParameter() {group=Flowmap2, parameterName=Noise UV Distortion})
  n70 = Divide(A=WorldPosition().XYZ, B=Multiply(A=ScalarParameter() {defaultValue=10, group=Flowmap2, parameterName=scale2}) {constB=100000})
  n71 = VectorParameter() {group=Flowmap2, parameterName=Flowmap2Channel, defaultValue=<rgba 1 0 0 0>}
  n72 = DotProduct(
   A=MaterialFunctionCall(Volume Texture=@n51, UVW=Add(A=@n69, B=@n70), Velocity=Add(A=@n57, B=@n64), Override Time=@n11.Time, Mip Level=@n65) {materialFunction=&Flowmaps_3D}.Result RGBA
   B=AppendVector(A=@n71.RGB, B=@n71.RGB)
  )
  n73 = Power(
   Base=LinearInterpolate(A=@n67, B=OneMinus(None=@n72), Alpha=ScalarParameter() {defaultValue=0.75, parameterName=Internal Mask Balance})
   Exp=ScalarParameter() {defaultValue=3, parameterName=Texture Mask Power}
  )
  n74 = VectorParameter() {parameterName=Lightning Color, defaultValue=<rgba 2 1.8 2 0.3>}
  n75 = Multiply(A=@n74.RGB, B=@n74.RGB)
  n76 = CollectionParameter() {collection=&MPC_Clouds, parameterName=LightningMultiplier}
  Multiply(A=Multiply(A=Multiply(A=@n50, B=@n73), B=@n21), B=Multiply(A=Multiply(A=@n75, B=@n76), B=Divide(B=@n20)))
  ScalarParameter() {defaultValue=0.001, parameterName=Exinction}
  TextureSample(UVs=ComponentMask(None=@n5) {g=true, r=true}, Level=ScalarParameter() {parameterName=2D Mask Mip}) {
   mipValueMode=TMVM_MipLevel
   samplerSource=SSM_Wrap_WorldGroupSettings
   samplerType=SAMPLERTYPE_LinearColor
   texture=&RT_CloudMask_01
  }
  n77 = ScalarParameter() {group=CloudBase, parameterName=CloudNegativeTrimThreshold}
  n78 = Multiply(A=@n48, B=ScalarParameter() {defaultValue=1, group=CloudBase, parameterName=CloudNegativeTrimMult, sliderMax=10})
  n79 = CollectionParameter() {collection=&MPC_Clouds, parameterName=NoiseMultiplier}
  n80 = ScalarParameter() {defaultValue=0.2, group=Flowmap1, parameterName=Detail}
  n81 = Add(A=@n72, B=ScalarParameter() {defaultValue=-0.5, group=Flowmap2, parameterName=Flowmap2Bias, sliderMax=1, sliderMin=-1})
  n82 = ScalarParameter() {defaultValue=0.25, group=Flowmap2, parameterName=Detail2}
  n83 = Multiply(
   A=Add(A=Multiply(A=@n68, B=@n80), B=Multiply(A=@n81, B=@n82))
   B=LinearInterpolate(A=ScalarParameter() {defaultValue=0.3, group=Flowmap1, parameterName=Bottom Noise}, Alpha=Saturate(None=@n37))
  )
  VolumetricAdvancedMaterialOutput!(
   PhaseG=ScalarParameter() {parameterName=Phase A}
   PhaseG2=ScalarParameter() {defaultValue=0.4, parameterName=Phase B}
   PhaseBlend=ScalarParameter() {defaultValue=0.5, parameterName=Phasey Phasey}
   MultiScatteringContribution=ScalarParameter() {defaultValue=0.4, parameterName=MS C}
   MultiScatteringOcclusion=Multiply(A=ScalarParameter() {defaultValue=0.5, parameterName=MS Occ}, B=ScalarParameter() {defaultValue=0.5, parameterName=Powertest})
   MultiScatteringEccentricity=ScalarParameter() {defaultValue=0.5, parameterName=MS Ec}
   ConservativeDensity=Add(A=If(A=@n48, B=@n77, "A > B"=@n48, "A < B"=@n78), B=Multiply(A=@n79, B=@n83))
  ) {multiScatteringApproximationOctaveCount=2}
 ]
}
