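# Volumetric cloud material that shapes soft cloud layers from a painted render-target mask and a height profile texture.
# World-projected mask coverage passes through exponential softening, then indexes the profile texture alongside normalized layer altitude.
# Two Voronoi volume samples add detail, the first able to distort the second's UVs.
# The re-softened density is scaled by tiling 2D noise, squared tint and altitude, with dual-phase, two-octave multi-scattering lighting.
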Material /Volumetrics/Content/Sky/Materials/M_VolumetricCloud_02_Profiles_Soft
graph {
 properties {
  bUsedWithVolumetricCloud=true
  blendMode=BLEND_Additive
  materialDomain=MD_Volume
  pixelDepthOffsetMode=PDOM_Legacy
  usedShadingModels=Default Lit
 }
 nodes [
  n1 = CloudSampleAttribute()
  n2 = Divide(A=WorldPosition().XYZ, B=Multiply(A=ScalarParameter() {defaultValue=400, parameterName=Mask Scale}) {constB=100000})
  n3 = VectorParameter() {parameterName=Offset, defaultValue=<rgba 0.5 0.5 0 0>}
  n4 = TextureSample!(
   UVs=ComponentMask(None=Subtract(A=@n2, B=@n3.RGB)) {g=true, r=true}
   Tex
   Level=ScalarParameter() {parameterName=2D Mask Mip}
   Apply View MipBias
  ) {mipValueMode=TMVM_MipLevel, samplerType=SAMPLERTYPE_LinearColor, texture=/Volumetrics/Tools/CloudCompositing/RenderTargets/RT_CloudMask_01}
  n5 = ScalarParameter() {defaultValue=0.4, parameterName=Bias}
  MaterialExpressionMax_0 = Max(A=Multiply(A=Subtract(A=@n4.R, B=@n5), B=ScalarParameter() {defaultValue=4, parameterName=Density})) {constB=0}
  n6 = Custom!(x=@MaterialExpressionMax_0) {
   code="return 1-exp(-x);"
   outputType=CMOT_Float1
   inputs [{inputName=x, input {expression=&self:MaterialExpressionMax_0}}]
  }
  n7 = Constant() {r=0.0059}
  n8 = TextureSample(UVs=Clamp(None=AppendVector(A=@n1.NormAltitudeInLayer, B=@n6), Min=@n7, Max=OneMinus(None=@n7)), Level=Constant()) {mipValueMode=TMVM_MipLevel, samplerType=SAMPLERTYPE_LinearColor, texture=/Volumetrics/Tools/CloudCompositing/Textures/Profile_01}
  n9 = ScalarParameter() {defaultValue=128, parameterName=scale}
  n10 = TextureSample(UVs=Divide(A=WorldPosition().XYZ, B=Multiply(A=@n9) {constB=100000}), Level=ScalarParameter() {defaultValue=3, parameterName=Volume Mip}) {mipValueMode=TMVM_MipLevel, samplerType=SAMPLERTYPE_LinearColor, texture=/Volumetrics/Content/VolumeTextures/Textures/VT_Voronoi}
  n11 = Subtract(A=@n10.R) {constB=0.5}
  n12 = Multiply(A=@n6, B=Multiply(A=@n11, B=ScalarParameter() {parameterName=Detail}))
  n13 = ScalarParameter() {parameterName=Noise UV Distortion}
  n14 = WorldPosition()
  n15 = Multiply(A=ScalarParameter() {defaultValue=32, parameterName=scale2}) {constB=100000}
  n16 = TextureSample(UVs=Add(A=Multiply(A=@n11, B=@n13), B=Divide(A=@n14.XYZ, B=@n15)), Level=ScalarParameter() {defaultValue=1, parameterName=Volume Mip2}) {mipValueMode=TMVM_MipLevel, samplerType=SAMPLERTYPE_LinearColor, texture=&VT_Voronoi}
  n17 = Subtract(A=@n16.R) {constB=0.5}
  n18 = ScalarParameter() {defaultValue=0.3, parameterName=Detail2}
  MaterialExpressionMultiply_15 = Multiply(A=Add(A=Add(A=@n8.R, B=@n12), B=Multiply(A=@n17, B=@n18)), B=ScalarParameter() {defaultValue=8, parameterName=DDD})
  n19 = Custom(x=@MaterialExpressionMultiply_15) {
   code="return 1-exp(-x);"
   outputType=CMOT_Float1
   inputs [{inputName=x, input {expression=&self:MaterialExpressionMultiply_15}}]
  }
  n20 = ScalarParameter() {defaultValue=250, parameterName=3}
  n21 = ComponentMask(None=Divide(A=WorldPosition().XYZ, B=Multiply(A=@n20) {constB=100000})) {g=true, r=true}
  n22 = ScalarParameter() {parameterName=2D Mask Mip2}
  n23 = LinearInterpolate(
   A=ScalarParameter() {parameterName=OutD}
   B=ScalarParameter() {defaultValue=0.2, parameterName=OutD2}
   Alpha=Power(
    Base=TextureSample(UVs=@n21, Level=@n22) {mipValueMode=TMVM_MipLevel, texture=/Engine/MapTemplates/TilingNoise05}.R
    Exp=ScalarParameter() {defaultValue=2.5, parameterName=Param_1}
   )
  )
  n24 = VectorParameter() {parameterName=ecx, defaultValue=<rgba 0.5 0.575 0.65 0.5>}
  Multiply(A=Multiply(A=@n19, B=Max(A=@n23) {constB=0}), B=Multiply(A=Multiply(A=@n24.RGB, B=@n24.RGB), B=CloudSampleAttribute().NormAltitudeInLayer))
  VectorParameter() {parameterName=cc, defaultValue=<rgba 1 1 1 0>}
  VolumetricAdvancedMaterialOutput!(
   PhaseG=ScalarParameter() {parameterName=Phase A}
   PhaseG2=ScalarParameter() {defaultValue=0.8, parameterName=Phase B}
   PhaseBlend=ScalarParameter() {defaultValue=0.5, parameterName=Phasey Phasey}
   MultiScatteringContribution=ScalarParameter() {defaultValue=0.6, parameterName=MS C}
   MultiScatteringOcclusion=Multiply(A=ScalarParameter() {defaultValue=0.5, parameterName=MS Occ}, B=ScalarParameter() {defaultValue=0.7, parameterName=Powertest})
   MultiScatteringEccentricity=ScalarParameter() {defaultValue=0.3, parameterName=MS Ec}
   ConservativeDensity
  ) {multiScatteringApproximationOctaveCount=2}
 ]
}
