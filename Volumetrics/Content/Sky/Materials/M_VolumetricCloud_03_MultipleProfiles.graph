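# Volumetric cloud material that builds cloud shapes from a painted world-space layout mask and multiple height profiles.
# Biased mask channels index Profiles_Multi against layer altitude, separately or combined by static switch; the final max uses only red and blue.
# Two scales of Perlin-Worley noise add detail to this profile field, which is then scaled by altitude band, density and inverted mask alpha.
# Advanced volumetric output sets dual-lobe phase and two-octave multi-scattering.
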Material /Volumetrics/Content/Sky/Materials/M_VolumetricCloud_03_MultipleProfiles
graph {
 properties {
  bUsedWithVolumetricCloud=true
  blendMode=BLEND_Additive
  materialDomain=MD_Volume
  pixelDepthOffsetMode=PDOM_Legacy
  usedShadingModels=Default Lit
 }
 nodes [
  n1 = Constant()
  n2 = If(A=CloudSampleAttribute().NormAltitudeInLayer, B=Constant() {r=0.5}, "A > B"=@n1, "A == B"=Constant() {r=1}, "A < B"=@n1) {equalsThreshold=0.48}
  n3 = CloudSampleAttribute()
  n4 = Divide(
   A=WorldPosition().XYZ
   B=MaterialFunctionCall!(Km=ScalarParameter() {defaultValue=500, parameterName=SkyTextureScaleKm}) {materialFunction=/Engine/Functions/Engine_MaterialFunctions01/Units/Km-to-Cm}.Cm
  )
  n5 = TextureSampleParameter2D!(UVs=ComponentMask(None=Subtract(A=@n4) {constB=0.5}) {g=true, r=true}, Level, Apply View MipBias) {
   desc=2D Paint Mask
   mipValueMode=TMVM_MipLevel
   parameterName=CloudLayout
   samplerType=SAMPLERTYPE_LinearColor
   texture=/Volumetrics/Tools/CloudCompositing/RenderTargets/RT_CloudMask_01
  }
  n6 = Max(A=Max(A=@n5.R, B=@n5.R), B=@n5.B)
  n7 = ScalarParameter() {defaultValue=0.3, parameterName=Mask Bias}
  n8 = Divide(
   A=Subtract(A=StaticSwitchParameter(True=@n5.RGB, False=@n6) {defaultValue=true, parameterName=Sample Profiles Separately}, B=@n7)
   B=OneMinus(None=@n7)
  )
  n9 = MaterialFunctionCall(Float3=Multiply(A=ScalarParameter() {defaultValue=1, parameterName=Cloud UV X Scale}, B=@n8)) {materialFunction=&BreakOutFloat3Components}
  n10 = TextureSample!(UVs=AppendVector(A=@n3.NormAltitudeInLayer, B=@n9.R), Tex, Level, Apply View MipBias) {
   constMipValue=0
   desc=Cloud Profile
   mipValueMode=TMVM_MipLevel
   samplerType=SAMPLERTYPE_LinearColor
   texture=/Volumetrics/Tools/CloudCompositing/Textures/Profiles_Multi
  }
  n11 = TextureSample(UVs=AppendVector(A=@n3.NormAltitudeInLayer, B=@n9.G)) {
   constMipValue=0
   mipValueMode=TMVM_MipLevel
   samplerType=SAMPLERTYPE_LinearColor
   texture=&Profiles_Multi
  }
  n12 = TextureSample(UVs=AppendVector(A=@n3.NormAltitudeInLayer, B=@n9.B)) {
   constMipValue=0
   mipValueMode=TMVM_MipLevel
   samplerType=SAMPLERTYPE_LinearColor
   texture=&Profiles_Multi
  }
  n13 = MaterialFunctionCall(X=@n10.R, Y=@n11.G, Z=@n12.B) {materialFunction=&MakeFloat3}
  n14 = Max(A=Multiply(A=@n8) {constB=-1}) {constB=0}
  n15 = Multiply(
   A=MaterialFunctionCall(Vector=@n5.RGB, Default=Constant3Vector() {constant=<rgba 1 0 0 0>}) {materialFunction=&SafeNormalize}.Result
   B=@n10.RGB
  )
  n16 = MaterialFunctionCall(Float3=StaticSwitchParameter(True=Subtract(A=@n13.Result, B=@n14), False=Subtract(A=@n15, B=@n14)) {defaultValue=true, parameterName=Sample Profiles Separately}) {materialFunction=&BreakOutFloat3Components}
  n17 = Max(A=@n16.R, B=@n16.R)
  n18 = WorldPosition()
  n19 = Multiply(A=ScalarParameter() {defaultValue=8, parameterName=scale}) {constB=100000}
  n20 = Subtract(A=TextureSample(UVs=Divide(A=@n18.XYZ, B=@n19), Level=ScalarParameter() {parameterName=Volume Mip}) {mipValueMode=TMVM_MipLevel, texture=/Volumetrics/Content/VolumeTextures/Textures/VT_PerlinWorley_Balanced}.R) {constB=0.5}
  n21 = ScalarParameter() {defaultValue=0.3, parameterName=Detail}
  n22 = WorldPosition()
  n23 = Multiply(A=ScalarParameter() {defaultValue=12, parameterName=scale2}) {constB=100000}
  n24 = Subtract(
   A=TextureSample(UVs=Divide(A=@n22.XYZ, B=@n23), Level=ScalarParameter() {parameterName=Volume Mip}) {mipValueMode=TMVM_MipLevel, texture=&VT_PerlinWorley_Balanced}.G
   B=ScalarParameter() {defaultValue=0.5, parameterName=Noise 2 Bias}
  ) {constB=0.5}
  n25 = ScalarParameter() {defaultValue=0.2, parameterName=Detail2}
  n26 = Add(
   A=Subtract(A=Max(A=@n17, B=@n16.B), B=ScalarParameter() {defaultValue=0.5, parameterName=Distance Field Center})
   B=Add(A=Multiply(A=@n20, B=@n21), B=Multiply(A=@n24, B=@n25))
  )
  n27 = Divide(A=ScalarParameter() {defaultValue=64, parameterName=Cloud Density}) {constB=100}
  n28 = OneMinus(None=@n5.A)
  n29 = VectorParameter() {parameterName=ecx, defaultValue=<rgba 0.5 0.5 0.5 0.5>}
  Multiply(
   A=Multiply(A=Multiply(A=@n2, B=@n26), B=Multiply(A=@n27, B=@n28))
   B=Multiply(A=Multiply(A=@n29.RGB, B=@n29.RGB), B=CloudSampleAttribute().NormAltitudeInLayer)
  )
  VectorParameter() {parameterName=Albedo, defaultValue=<rgba 1 1 1 0>}
  VolumetricAdvancedMaterialOutput!(
   PhaseG=ScalarParameter() {defaultValue=0.5, group=Scattering and Phase, parameterName=Phase A}
   PhaseG2=ScalarParameter() {defaultValue=0.8, group=Scattering and Phase, parameterName=Phase B}
   PhaseBlend=ScalarParameter() {defaultValue=0.4, group=Scattering and Phase, parameterName=Phasey Phasey}
   MultiScatteringContribution=ScalarParameter() {defaultValue=0.5, group=Scattering and Phase, parameterName=MS C}
   MultiScatteringOcclusion=ScalarParameter() {defaultValue=0.2, group=Scattering and Phase, parameterName=MS Occ}
   MultiScatteringEccentricity=ScalarParameter() {defaultValue=0.5, group=Scattering and Phase, parameterName=MS Ec}
   ConservativeDensity
  ) {multiScatteringApproximationOctaveCount=2}
 ]
}
