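# Additive volumetric cloud material that shapes billowy clouds from a painted-style profile texture and 3D Voronoi noise.
# The CloudCompositing mask render target, mapped in world space, pairs with normalized layer altitude to look up the profile distance field.
# Two world-scaled Voronoi samples, faded in by altitude, offset that field before density, squared tint and altitude scaling.
# The advanced volumetric output sets dual-lobe phase and two-octave multi-scattering.
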Material /Volumetrics/Content/Sky/Materials/M_VolumetricCloud_03_Profiles_Billowy
graph {
 properties {
  bUsedWithVolumetricCloud=true
  blendMode=BLEND_Additive
  materialDomain=MD_Volume
  pixelDepthOffsetMode=PDOM_Legacy
  usedShadingModels=Default Lit
 }
 nodes [
  n1 = CloudSampleAttribute()
  n2 = Divide(
   A=WorldPosition().XYZ
   B=Multiply(A=CollectionParameter() {collection=/Volumetrics/Tools/CloudCompositing/MPC/MPC_Clouds, parameterName=SkyTextureScaleKm}) {constB=100000}
  )
  n3 = TextureSample!(UVs=ComponentMask(None=Subtract(A=@n2) {constB=0.5}) {g=true, r=true}, Tex, Level, Apply View MipBias) {
   constMipValue=0
   mipValueMode=TMVM_MipLevel
   samplerType=SAMPLERTYPE_LinearColor
   texture=/Volumetrics/Tools/CloudCompositing/RenderTargets/RT_CloudMask_01
  }
  n4 = Multiply(A=Subtract(A=@n3.R, B=ScalarParameter() {parameterName=Mask Bias}), B=ScalarParameter() {defaultValue=1, parameterName=Cloud UV X Scale})
  n5 = Divide() {constA=0.5, constB=256}
  n6 = TextureSample(UVs=Clamp(None=AppendVector(A=@n1.NormAltitudeInLayer, B=@n4), Min=@n5, Max=OneMinus(None=@n5))) {
   constMipValue=0
   mipValueMode=TMVM_MipLevel
   samplerType=SAMPLERTYPE_LinearColor
   texture=/Volumetrics/Tools/CloudCompositing/Textures/Profiles_Multi
  }
  n7 = Subtract(A=@n6.R, B=ScalarParameter() {defaultValue=0.5, parameterName=Distance Field Center})
  n8 = ScalarParameter() {defaultValue=64, parameterName=scale}
  n9 = TextureSample(UVs=Divide(A=WorldPosition().XYZ, B=Multiply(A=@n8) {constB=100000}), Level=ScalarParameter() {parameterName=Volume Mip}) {mipValueMode=TMVM_MipLevel, samplerType=SAMPLERTYPE_LinearColor, texture=/Volumetrics/Content/VolumeTextures/Textures/VT_Voronoi_Simple}
  n10 = Subtract(A=@n9.R) {constB=0.5}
  n11 = ScalarParameter() {parameterName=Detail}
  n12 = ScalarParameter() {defaultValue=68, parameterName=scale2}
  n13 = TextureSample(UVs=Divide(A=WorldPosition().XYZ, B=Multiply(A=@n12) {constB=100000}), Level=ScalarParameter() {parameterName=Volume Mip}) {mipValueMode=TMVM_MipLevel, samplerType=SAMPLERTYPE_LinearColor, texture=&VT_Voronoi_Simple}
  n14 = Subtract(A=@n13.R, B=ScalarParameter() {defaultValue=0.1, parameterName=Noise 2 Bias}) {constB=0.5}
  n15 = ScalarParameter() {defaultValue=-0.2, parameterName=Detail2}
  n16 = Multiply(A=CloudSampleAttribute().NormAltitudeInLayer, B=Add(A=Multiply(A=@n10, B=@n11), B=Multiply(A=@n14, B=@n15)))
  n17 = ScalarParameter() {defaultValue=512, parameterName=Cloud Density}
  n18 = VectorParameter() {parameterName=ecx, defaultValue=<rgba 0.5 0.6 0.7 0.5>}
  Multiply(
   A=Multiply(A=Add(A=@n7, B=@n16), B=Divide(A=@n17) {constB=100})
   B=Multiply(A=Multiply(A=@n18.RGB, B=@n18.RGB), B=CloudSampleAttribute().NormAltitudeInLayer)
  )
  VectorParameter() {parameterName=Albedo, defaultValue=<rgba 1 1 1 0>}
  VolumetricAdvancedMaterialOutput!(
   PhaseG=ScalarParameter() {defaultValue=0.5, group=Scattering and Phase, parameterName=Phase A}
   PhaseG2=ScalarParameter() {defaultValue=0.8, group=Scattering and Phase, parameterName=Phase B}
   PhaseBlend=ScalarParameter() {defaultValue=0.4, group=Scattering and Phase, parameterName=Phasey Phasey}
   MultiScatteringContribution=ScalarParameter() {defaultValue=0.5, group=Scattering and Phase, parameterName=MS C}
   MultiScatteringOcclusion=ScalarParameter() {defaultValue=0.2, group=Scattering and Phase, parameterName=MS Occ}
   MultiScatteringEccentricity=ScalarParameter() {defaultValue=0.5, group=Scattering and Phase, parameterName=MS Ec}
   ConservativeDensity
  ) {multiScatteringApproximationOctaveCount=2}
 ]
}
