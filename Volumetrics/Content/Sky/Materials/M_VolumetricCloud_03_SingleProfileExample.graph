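# Example volumetric cloud material that shapes a cloud layer from one profile lookup and a 2D paint mask.
# World XY, scaled by MPC_Clouds, samples the mask render target, whose red channel pairs with normalized layer altitude to index the profile texture.
# The profile, offset by a center value, is scaled by density, a squared color and altitude.
# An advanced volumetric output sets dual-lobe phase and two-octave multi-scattering.
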Material /Volumetrics/Content/Sky/Materials/M_VolumetricCloud_03_SingleProfileExample
graph {
 properties {
  bUsedWithVolumetricCloud=true
  blendMode=BLEND_Additive
  materialDomain=MD_Volume
  pixelDepthOffsetMode=PDOM_Legacy
  usedShadingModels=Default Lit
 }
 nodes [
  n1 = Divide(
   A=WorldPosition().XYZ
   B=Multiply(A=CollectionParameter() {collection=/Volumetrics/Tools/CloudCompositing/MPC/MPC_Clouds, parameterName=SkyTextureScaleKm}) {constB=100000}
  )
  n2 = TextureSample!(UVs=ComponentMask(None=Subtract(A=@n1) {constB=0.5}) {g=true, r=true}, Tex, Level, Apply View MipBias) {
   constMipValue=0
   desc=2D Paint Mask
   mipValueMode=TMVM_MipLevel
   samplerType=SAMPLERTYPE_LinearColor
   texture=/Volumetrics/Tools/CloudCompositing/RenderTargets/RT_CloudMask_01
  }
  n3 = TextureSample(UVs=AppendVector(A=CloudSampleAttribute().NormAltitudeInLayer, B=@n2.R)) {
   constMipValue=0
   desc=Cloud Profile
   mipValueMode=TMVM_MipLevel
   samplerType=SAMPLERTYPE_LinearColor
   texture=/Volumetrics/Tools/CloudCompositing/Textures/RT_Profiles
  }
  n4 = ScalarParameter() {defaultValue=0.5, parameterName=Distance Field Center}
  n5 = ScalarParameter() {defaultValue=200, parameterName=Cloud Density}
  n6 = VectorParameter() {parameterName=ecx, defaultValue=<rgba 0.5 0.6 0.7 0.5>}
  Multiply(
   A=Multiply(A=Subtract(A=@n3.R, B=@n4), B=Divide(A=@n5) {constB=100})
   B=Multiply(A=Multiply(A=@n6.RGB, B=@n6.RGB), B=CloudSampleAttribute().NormAltitudeInLayer)
  )
  VectorParameter() {parameterName=Albedo, defaultValue=<rgba 1 1 1 0>}
  VolumetricAdvancedMaterialOutput!(
   PhaseG=ScalarParameter() {defaultValue=0.5, group=Scattering and Phase, parameterName=Phase A}
   PhaseG2=ScalarParameter() {defaultValue=0.8, group=Scattering and Phase, parameterName=Phase B}
   PhaseBlend=ScalarParameter() {defaultValue=0.4, group=Scattering and Phase, parameterName=Phasey Phasey}
   MultiScatteringContribution=ScalarParameter() {defaultValue=0.5, group=Scattering and Phase, parameterName=MS C}
   MultiScatteringOcclusion=ScalarParameter() {defaultValue=0.2, group=Scattering and Phase, parameterName=MS Occ}
   MultiScatteringEccentricity=ScalarParameter() {defaultValue=0.5, group=Scattering and Phase, parameterName=MS Ec}
   ConservativeDensity
  ) {multiScatteringApproximationOctaveCount=2}
 ]
}
